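; Quest gameplay ability for Season 10 Week 6 Quest 7, crediting the DanceBrute objective for the Ostrich (BRUTE) vehicle.
; On event activation it calls the parent, then checks the owning pawn's tags for the Ostrich vehicle tag.
; On a match it sends the DanceBrute stat event to the owner's quest manager and to that of the pawn in seat 0.
; The ability ends either way.
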(event Ability|EventActivateAbilityFromEvent
  (params (EventData :type Gameplay%20Event%20Data%20Structure%20%28by%20ref%29))
  (bind _owningplayerpawn (Variables|GAVars|Data|GetOwningPlayerPawn))
  (|CallParentFunction EventData)
  (if (GameplayTags|TagContainer|CompareTagAssetInterfacetoOtherTags :TagCase_0 (Variables|Default|GetT_VehicleOstrich) :Gameplay%20Tag%20Asset%20Interface _owningplayerpawn)
    (Quest|SendComplexCustomStatEvent (Variables|GAVars|Data|GetOwningPlayerQuestManager) 0 "(GameplayTags=((TagName=\"Athena.Quests.DanceBrute\")))" "(GameplayTags=)" 1)
    (bind _returnvalue (Vehicle|GetPawnatSeat(Message) _owningplayerpawn 0))
    (Quest|SendComplexCustomStatEvent (Quests|GetQuestManager (Fortnite|GetFortPlayerControllerfromActor _returnvalue) "Invalid") 0 "(GameplayTags=((TagName=\"Athena.Quests.DanceBrute\")))" "(GameplayTags=)" 1)
    (Ability|EndAbility)
    (else
      (Ability|EndAbility))))
